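# Parent material for the pages of the WordUnity back bling, built on the 2023 anime character shader with Cloth shading and material attributes.
# The shader's Pre, Calculations and Post functions run in sequence.
# Between them, the FX red channel, softened by a power and limited to front faces, blends lit and shaded colors toward BaseColor and drives tinted emissive.
# An IsKicks switch optionally adds MF_Sock, while the eye animation branch feeds nothing.
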Material /BRCosmetics/Accessories/FORT_Backpacks/Backpack_F_MED_WordUnity/Materials/M_WordUnity_Backpack_Pages_Parent
graph {
 properties {
  bAutomaticallySetUsageInEditor=false
  bUseMaterialAttributes=true
  bUsedWithClothing=true
  bUsedWithMeshDeformer=true
  bUsedWithMorphTargets=true
  bUsedWithNiagaraMeshParticles=true
  bUsedWithSkeletalMesh=true
  bUsedWithStaticLighting=true
  pixelDepthOffsetMode=PDOM_Legacy
  shadingModel=MSM_Cloth
  usedShadingModels=Cloth
 }
 nodes [
  n1 = MaterialFunctionCall!(CustomUV) {materialFunction=/Game/Global/Materials/Character_Parent_2023/Anime_Parent_2023/MaterialFunctions/MFC_AnimeShader_2023_Pre}
  n2 = VertexColor()
  n3 = TextureSampleParameter2D() {
   group=01_EyeAnimation_Parameters
   parameterName=EyeMask
   samplerSource=SSM_Wrap_WorldGroupSettings
   samplerType=SAMPLERTYPE_Masks
   sortPriority=-32
   texture=&T_WhiteMask
  }
  n4 = StaticSwitchParameter(True=AppendVector(A=@n2, B=@n2), False=@n3.RGBA) {group=01_EyeAnimation_Parameters, parameterName=UseVertexColorForEyeMask, sortPriority=-33}
  n5 = TextureCoordinate() {coordinateIndex=1}
  n6 = Sign(None=Subtract(A=ComponentMask(None=@n5) {r=true}) {constB=0.0001})
  MaterialFunctionCall!(
   Color Lit=@n1.Color_Lit
   Color Shaded=@n1.Color_Lit
   Use UV2=StaticBoolParameter() {defaultValue=true, group=01_EyeAnimation_Parameters, parameterName=Use UV2 for Eye Animation, sortPriority=-30}
   Mask=StaticSwitchParameter(
    True=ChannelMaskParameter(None=@n4) {group=01_EyeAnimation_Parameters, parameterName=EyeMaskChannel, sortPriority=-32}.R
    False=Saturate(None=@n6)
   ) {group=01_EyeAnimation_Parameters, parameterName=UseEyeMask, sortPriority=-33}
  ) {materialFunction=/Game/Global/Materials/Character_Parent_2023/Anime_Parent_2023/MaterialFunctions/MF_AnimeEyeAnimation_v2}
  n7 = MaterialFunctionCall!(CustomLightVector, OverrideNormal_WS, STT=@n1.STT, Normal_Map=@n1.Normal, DFL=@n1.DFL, FX=@n1.FX) {materialFunction=/Game/Global/Materials/Character_Parent_2023/Anime_Parent_2023/MaterialFunctions/MFC_AnimeShader_2023_Calculations}
  n8 = VectorParameter() {parameterName=BaseColor, defaultValue=<rgba 0.8148 0.2582 1 1>}
  n9 = Multiply(A=Power(Base=ComponentMask(None=@n1.FX) {r=true}, Exp=Constant() {r=0.6}), B=Saturate(None=TwoSidedSign()))
  n10 = MaterialFunctionCall!(
   Shading_Gradient=@n7.Shading_Gradient
   Threshold_Gradient=@n7.Shading_Gradient
   Specular_Gradient=@n7.Shading_Gradient
   Sheen_Mask=@n7.Shading_Gradient
   Thickness=@n7.Shading_Gradient
   DFL_Mask=@n7.Shading_Gradient
   FX=@n7.Shading_Gradient
   Metallic_Gradient=@n7.Shading_Gradient
   RimLight_Gradient=@n7.Shading_Gradient
   NormalBuffer=@n7.Shading_Gradient
   --------
   CINE_LightSources=@n7.Shading_Gradient
   CINE_LightColor=@n7.Shading_Gradient
   ---------
   Color_Lit=LinearInterpolate(A=@n1.Color_Lit, B=@n8.RGB, Alpha=@n9)
   Color_Shaded=LinearInterpolate(A=@n1.Color_Shaded, B=@n8.RGB, Alpha=@n9)
   SRM=@n1.SRM
   Emissive=Multiply(A=@n9, B=VectorParameter() {parameterName=Emissive, defaultValue=<rgba 0.8148 0.2582 1 1>}.RGB)
   DFL_Color=@n1.SRM
   Specular_Color=@n1.SRM
   RimLight_Color=@n1.SRM
  ) {materialFunction=/Game/Global/Materials/Character_Parent_2023/Anime_Parent_2023/MaterialFunctions/MFC_AnimeShader_2023_Post}
  StaticSwitchParameter(
   True=MaterialFunctionCall!(InOpacity, Material=@n10.Result, Mask) {materialFunction=/Game/Global/Materials/Cosmetics/Kicks/Socks/Materials/MF_Sock}.Result
   False=@n10.Result
  ) {group=Kicks, parameterName=IsKicks}
  TemporalResponsivenessOutput!(None)
 ]
}
